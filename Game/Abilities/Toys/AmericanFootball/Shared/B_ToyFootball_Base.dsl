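; Base throwable American football toy that scores a pass when it touches a pawn other than its thrower.
; In flight it spirals with audio, bouncing along randomized mirrored velocity on fast impacts until a slow hit ends the spiral and trail.
; A catch before then plays a score cue, destroys the ball and builds a replicated message, which clients print to the feed with wording and audience based on pass distance.
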(event Visuals|EventInitializeToyPreviewInFrontend
  (|CallParentFunction)
  (Components|Activation|SetActive (Variables|Default|GetFootballTrail) false false))

(event Custom|ClearBounce
  (Variables|Default|SetBounce false))

(event Collision|EventHit
  (params (MyComp :type Primitive%20Component%20Object%20Reference) (Other :type Actor%20Object%20Reference) (OtherComp :type Primitive%20Component%20Object%20Reference) (bSelfMoved :type Boolean) (HitLocation :type Vector :default "0, 0, 0") (HitNormal :type Vector :default "0, 0, 0") (NormalImpulse :type Vector :default "0, 0, 0") (Hit :type Hit%20Result%20Structure%20%28by%20ref%29))
  (bind _projectilemovement (Variables|Default|GetProjectileMovement))
  (bind _velocity (Variables|/Script/Engine.MovementComponent|GetVelocity _projectilemovement))
  ;[1 1 1] On hit - give a random impulse based on the hit normal
  (|CallParentFunction MyComp Other OtherComp bSelfMoved HitLocation HitNormal NormalImpulse Hit)
  (Utilities|FlowControl|Sequence
    (:then_0
      (if (not (Variables|Default|GetBounce))
        (Variables|Default|SetBounce true)
        (Utilities|Time|SetTimerbyEvent (delegate Custom|ClearBounce) 0.5) ; Avoid multi-bouncing
        (if (>= (Math|Vector|VectorLength _velocity) 4e+02)
          (Variables|/Script/Engine.MovementComponent|SetVelocity (Math|Vector|ClampVectorSize (* (Math|Vector|MirrorVectorbyNormal _velocity (+ HitLocation (Math|Random|RandomUnitVector))) 0.7) 0.0 2.4e+03) _projectilemovement)
          (Utilities|FlowControl|DoOnce false
            (:Completed
              (Variables|Default|SetHit true) ; Hit Variable
              (Components|Activation|SetActive (Variables|Default|GetFootballTrail) false true))))))
    (:then_1
      (Utilities|FlowControl|DoOnce false
        (:Completed
          (Audio|Components|Audio|FadeOut (Variables|Default|GetInAirAudio) 0.25 0.0 "Linear" :!class /Script/Engine.AudioComponent))))))

(event Custom|PlayScoringAnim
  (bind (_returnvalue_location _returnvalue_rotation) (Math|Transform|BreakTransform (Transformation|GetWorldTransform (Variables|FortPhysicsBall|GetSphereCollision))))
  ;[1 1 1] Celebration FX
  (Variables|/Script/Engine.ProjectileMovementComponent|SetbSimulationEnabled false (Variables|Default|GetProjectileMovement))
  (Collision|SetActorEnableCollision false)
  (Utilities|FlowControl|Sequence
    (:then_0
      (Networking|SwitchHasAuthority
        (:Authority
          (Utilities|FlowControl|Delay (Variables|Default|GetLerpDuration))
          (Gameplay|SilentDie false))))
    (:then_1
      (Components|Timeline|SetPlayRate (Variables|B_ToyFootball_Base|GetLerpIntoTouchdown) (/ 1.0 (Variables|Default|GetLerpDuration)))
      (bind (_direction _position _scale) (Timeline|LerpIntoTouchdown PlayFromStart)
        (:Update
          (Transformation|SetWorldTransform (Variables|FortPhysicsBall|GetSphereCollision) (Math|Transform|MakeTransform (Math|Vector|Lerp(Vector) (Variables|Default|GetIdealLerpStartPos) (Variables|Default|GetIdealLerpEndPos) _position) _returnvalue_rotation (Math|Conversions|ToVector(Float) _scale)) false false))))))

(event Collision|EventActorBeginOverlap
  (params (OtherActor :type Actor%20Object%20Reference))
  (|CallParentFunction OtherActor)
  (CallFunction|HitPlayer/VehicleBounceSound))

(event Visuals|EventInitializeToyInstance
  (params (OwningPC :type Fort%20Player%20Controller%20Object%20Reference) (NumTimesSummoned :type Integer))
  (|CallParentFunction OwningPC NumTimesSummoned)
  (Variables|Default|SetHasFinishedSpawning true))

(event EventBeginPlay
  (|CallParentFunction)
  (Variables|Default|SetIs%20Throwing true))

(event EventTick
  (params (DeltaSeconds :type Float%20%28single-precision%29))
  (if (Variables|Default|GetHit)
    (|CallParentFunction DeltaSeconds) ; Becomes like any other toy after hit
    (else
      (Networking|CanExecuteCosmeticEvents
        (:True
          ;[1 1 1] Manually handle rotation until we hit something (this is the spiral)
          (Transformation|AddLocalRotation (Variables|Default|GetBall) "4.000000, 0, 0" false false))))))

(event EventOnPawnTouchedMe
  (params (TouchingPawn :type Pawn%20Object%20Reference) (ImpactNormal :type Vector :default "0, 0, 0") (OtherVelocity :type Vector :default "0, 0, 0"))
  ;[1 1 1] Call celebration FX and message when a pawn is hit
  (|CallParentFunction TouchingPawn ImpactNormal OtherVelocity)
  (bind _asfort_player_pawn (Utilities|Casting|CastToFortPlayerPawn TouchingPawn))
  (bind _asfort_player_state (Utilities|Casting|CastToFortPlayerState (Variables|/Script/Engine.Controller|GetPlayerState (Pawn|GetController _asfort_player_pawn))))
  (if (Math|Boolean|ANDBoolean (Utilities|NotEqual(Object) _asfort_player_state (Actor|GetOwner)) (not (Variables|Default|GetHit)) (Variables|Default|GetHasFinishedSpawning))
    (bind _meterstraveled (CallFunction|CalcDistanceTraveled (Transformation|GetActorLocation TouchingPawn)))
    (CallFunction|CreateMessage (Utilities|Casting|CastToFortPlayerState (Actor|GetOwner)) _asfort_player_state _meterstraveled)
    (Networking|SwitchHasAuthority
      (:Authority
        ;;[1 0.8413 0.08369] Play FX and kill on server
        (CallFunction|PlayScoringCueandDestruct (Transformation|GetActorLocation TouchingPawn) (+ (Transformation|GetActorLocation TouchingPawn) (* (Transformation|GetActorUpVector) -1.2e+02)))
        (Gameplay|SilentDie false))
      (:Remote
        ;;[0.1583 1 0.2487] Hide on client
        (Development|SetHiddeninGame (Variables|Default|GetBall) true false)))))

(event EventSetBallVelocityAndStartTicking
  (params (NewVelocity :type Vector :default "0, 0, 0"))
  (|CallParentFunction NewVelocity)
  (CallFunction|ClientonThrown))

(event Custom|Client%20On%20Thrown
  (Utilities|FlowControl|DoOnce false
    (:Completed
      (Variables|Default|SetIs%20Throwing false)
      (Audio|Components|Audio|FadeIn (Variables|Default|GetInAirAudio) 0.25 1.0 0.0 "Linear" :!class /Script/Engine.AudioComponent))))

(event Custom|Hit%20Player/Vehicle%20Bounce%20Sound
  ;[1 1 1] Support Collision with Players and Vehicles
  (if (not (Variables|Default|GetIs%20Throwing))
    (Utilities|FlowControl|DoOnce false
      (:Completed
        (Audio|PlaySoundatLocation "/Game/Athena/Sounds/Toys/Football/Toys_Football_Bounce_Cue.Toys_Football_Bounce_Cue" (Transformation|GetActorLocation))
        (Utilities|FlowControl|Delay 0.2)))))

(fn CalcDistanceTraveled
  (params (HitLocation :type Vector))
  (bind _output_get (Variables|Default|SetMetersTraveled (/ (Math|Vector|VectorLengthXY (- HitLocation (Variables|Default|GetInitialSpawnPosition))) 1e+02)))
  (return _output_get))

(fn CreateMessage
  (params (PlayerWhoThrew :type Fort%20Player%20State%20Object%20Reference) (PlayerWhoCaught :type Fort%20Player%20State%20Object%20Reference) (PassDistance :type Float%20%28double-precision%29))
  (if (Variables|/Script/FortniteGame.FortRuntimeOptions|GetbToyMessagingEnabled (Options|GetRuntimeOptions))
    (Variables|Default|SetDisplayMessage (Utilities|Struct|MakeStructFootballMinigameMessage PlayerWhoThrew PlayerWhoCaught PassDistance (Variables|Default|GetMessageSN)))
    (Variables|Default|SetMessageSN (Math|Byte|Byte+Byte (Variables|Default|GetMessageSN) 1))))

(fn OnRep_DisplayMessage
  (bind _returnvalue (BlueprintContext|GetFortToyContext))
  (bind _returnvalue_2 (BlueprintContext|GetFortToyContext))
  (bind (_displaymessage_playerwhothrew_22_ef5d780b4020e911ab429d9ba6b977ba _displaymessage_playerwhocaught_25_800cc3c44dd84e656fcddc9e415ec97d _displaymessage_passdistance_26_a600d2a048cb01620b914bb4107a6cd5) (Utilities|Struct|BreakStructFootballMinigameMessage (Variables|GetDisplayMessage_PlayerWhoThrew_22_EF5D780B4020E911AB429D9BA6B977BA)))
  (Networking|SwitchHasAuthority
    (:Remote
      (Utilities|IsValid _displaymessage_playerwhothrew_22_ef5d780b4020e911ab429d9ba6b977ba
        (:Is%20Valid
          (bind _returnvalue_3 (Class|FortToyContext|FormatPlayerNameforDisplay _returnvalue _displaymessage_playerwhothrew_22_ef5d780b4020e911ab429d9ba6b977ba "ToyOwner"))
          (bind _returnvalue_4 (Class|FortToyContext|FormatPlayerNameforDisplay _returnvalue _displaymessage_playerwhocaught_25_800cc3c44dd84e656fcddc9e415ec97d "OtherPlayerInteractingWithToy"))
          (bind _formatstring (CallFunction|PickFormatString))
          (Utilities|IsValid _returnvalue_2
            (:Is%20Valid
              (Class|FortToyContext|ConditionalPrinttoMessageFeed _returnvalue_2 _displaymessage_playerwhothrew_22_ef5d780b4020e911ab429d9ba6b977ba (Transformation|GetActorLocation) (select (> _displaymessage_passdistance_26_a600d2a048cb01620b914bb4107a6cd5 (Variables|Default|GetMinimumDistanceforTeamRelevance)) "SquadAndTeamMembers" "SquadMembersOnly") (Utilities|Text|FormatText _formatstring _returnvalue_3 _returnvalue_4 (Math|Float|Round _displaymessage_passdistance_26_a600d2a048cb01620b914bb4107a6cd5))))))))))

(fn OnRep_IdealLerpEndPos)

(fn PickFormatString
  (bind (_displaymessage_playerwhothrew_22_ef5d780b4020e911ab429d9ba6b977ba _displaymessage_playerwhocaught_25_800cc3c44dd84e656fcddc9e415ec97d _displaymessage_passdistance_26_a600d2a048cb01620b914bb4107a6cd5) (Utilities|Struct|BreakStructFootballMinigameMessage (Variables|GetDisplayMessage_PlayerWhoThrew_22_EF5D780B4020E911AB429D9BA6B977BA)))
  (if (<= _displaymessage_passdistance_26_a600d2a048cb01620b914bb4107a6cd5 9.9)
    (return (Variables|Default|GetMinimumDistanceMessage))
    (elif (Math|Float|InRange(Float) _displaymessage_passdistance_26_a600d2a048cb01620b914bb4107a6cd5 1e+01 2e+01)
      (return (Variables|Default|GetFirstDownMessage))
      (elif (Math|Float|InRange(Float) _displaymessage_passdistance_26_a600d2a048cb01620b914bb4107a6cd5 2e+01 4e+01)
        (return (Variables|Default|GetTouchdownMessage))
        (elif (>= _displaymessage_passdistance_26_a600d2a048cb01620b914bb4107a6cd5 4e+01)
          (return (Variables|Default|GetHailMaryMessage)))))))

(fn PlayScoringCueAndDestruct
  (params (Target%20Start%20Pos :type Vector) (Target%20End%20Pos :type Vector))
  (CallFunction|PlayCueCenteredonMe (Variables|Default|GetScoreCue))
  (Variables|Default|SetIdealLerpStartPos Target%20Start%20Pos)
  (Variables|Default|SetIdealLerpEndPos Target%20End%20Pos))

(fn ConstructionScript
  (|CallParentFunction))
